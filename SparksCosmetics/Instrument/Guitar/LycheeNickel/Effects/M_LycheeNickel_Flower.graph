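# Masked, lit material for flower Niagara sprites in the effects of the LycheeNickel guitar cosmetic.
# One flower texture packs its masks by channel: red is tinted by particle color, and green, raised to a low power to boost it, is tinted by an Outline Color parameter that defaults to black.
# The two tints are summed for color, while a separate sum of red and green-times-alpha forms a mask.
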Material /SparksCosmetics/Instrument/Guitar/LycheeNickel/Effects/M_LycheeNickel_Flower
graph {
 properties {bAutomaticallySetUsageInEditor=false, bUsedWithNiagaraSprites=true, blendMode=BLEND_Masked, usedShadingModels=Default Lit}
 nodes [
  n1 = TextureSample() {texture=&/T_LycheeNickel_Flower}
  Add(A=@n1.R, B=Multiply(A=@n1.G, B=@n1.A))
  Add(
   A=Multiply(A=ParticleColor().RGB, B=@n1.R)
   B=Multiply(A=VectorParameter() {parameterName=Outline Color, defaultValue=<rgba 0 0 0 1>}.RGB, B=Power(Base=@n1.G) {constExponent=0.1})
  )
 ]
}
